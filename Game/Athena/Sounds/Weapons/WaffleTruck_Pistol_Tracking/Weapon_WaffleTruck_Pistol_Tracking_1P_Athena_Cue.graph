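# First-person fire sound cue for the WaffleTruck tracking pistol in Athena.
# A random node picks one of four equally weighted fire waves and passes it through a default modulator into a mixer whose second input is empty.
# First-person weapon attenuation, a ranged-fire sound class and a gunfire concurrency set apply, alongside owner-limited, stop-oldest override values.
# The debug flag is left on.
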SoundCue /Game/Athena/Sounds/Weapons/WaffleTruck_Pistol_Tracking/Weapon_WaffleTruck_Pistol_Tracking_1P_Athena_Cue
root=Mixer_1
nodes [
 {id=Mixer_1, children [Modulator_0,null], properties {inputVolume [1,2]}}
 {id=Modulator_0, children [Random_1]}
 {id=Random_1, children [WavePlayer_0,WavePlayer_1,WavePlayer_4,WavePlayer_5], properties {weights [1,1,1,1]}}
 {id=WavePlayer_0, properties {soundWaveAssetPtr=&/Pistol_WaffleTruck_Tracking_Fire_01}}
 {id=WavePlayer_1, properties {soundWaveAssetPtr=&/Pistol_WaffleTruck_Tracking_Fire_02}}
 {id=WavePlayer_4, properties {soundWaveAssetPtr=&/Pistol_WaffleTruck_Tracking_Fire_03}}
 {id=WavePlayer_5, properties {soundWaveAssetPtr=&/Pistol_WaffleTruck_Tracking_Fire_04}}
]
properties {
 attenuationSettings=&Fort_Weapon_Projectile_FirstPerson_Athena_Att
 bDebug=true
 maxDistance=2097152
 soundClassObject=&Fort_WeaponRanged_Fire_1P_PSM
 volumeMultiplier=1.3
 concurrencyOverrides {bLimitToOwner=true, maxCount=4, resolutionRule=StopOldest}
 concurrencySet [&Weapon_Gunfire_FirstPerson_Conc]
}
